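# Sound cue for the deploy sound of the TacticalBiker cosmetic glider.
# A mixer plays the Open wave immediately alongside a Loop_In wave held back by a fixed three-second delay and mixed slightly louder.
# It uses glider open/close attenuation and concurrency, a glider-open audio property sheet, and the Fort_Gliders_Land sound class.
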SoundCue /BRCosmetics/Sounds/Gliders/TacticalBiker/Glider_TacticalBiker_Open_Cue
root=Mixer_0
nodes [
 {id=Delay_0, children [WavePlayer_0], properties {delayMax=3, delayMin=3}}
 {id=Mixer_0, children [WavePlayer_1,Delay_0], properties {inputVolume [1,1.2]}}
 {id=WavePlayer_0, properties {soundWaveAssetPtr=&/Glider_TacticalBiker_Loop_In_01}}
 {id=WavePlayer_1, properties {soundWaveAssetPtr=&/Glider_TacticalBiker_Open}}
]
properties {
 attenuationSettings=/Game/Sounds/Attenuation/Fort_Player_Glider_OpenClose_Athena_Att
 audioPropertiesSheet=/BRCosmetics/Sounds/Settings/PropertySheets/Glider/APS_Cosmetic_Glider_Open
 maxDistance=10000
 soundClassObject=&Fort_Gliders_Land
 concurrencySet [&Athena_GliderOpenClose_Conc]
}
