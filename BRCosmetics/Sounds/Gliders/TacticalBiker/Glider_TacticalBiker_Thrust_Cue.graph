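# Sound cue for the thrust sound of the TacticalBiker glider cosmetic.
# A Random node picks one of three equally weighted Loop_In wave variants, with no further processing nodes in the graph.
# The cue uses player glider thrust attenuation, a thrust concurrency set and the cosmetic glider loop audio property sheet, and plays at reduced volume under the Fort_Gliders_Land sound class.
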SoundCue /BRCosmetics/Sounds/Gliders/TacticalBiker/Glider_TacticalBiker_Thrust_Cue
root=Random_0
nodes [
 {id=Random_0, children [WavePlayer_1,WavePlayer_2,WavePlayer_3], properties {weights [1,1,1]}}
 {id=WavePlayer_1, properties {soundWaveAssetPtr=&/Glider_TacticalBiker_Loop_In_01}}
 {id=WavePlayer_2, properties {soundWaveAssetPtr=&/Glider_TacticalBiker_Loop_In_02}}
 {id=WavePlayer_3, properties {soundWaveAssetPtr=&/Glider_TacticalBiker_Loop_In_03}}
]
properties {
 attenuationSettings=&Fort_Player_Glider_Thrust_Athena_Att
 audioPropertiesSheet=&APS_Cosmetic_Glider_Loop
 maxDistance=12500
 soundClassObject=&Fort_Gliders_Land
 volumeMultiplier=0.65
 concurrencySet [&Athena_Glider_Thrust_Conc]
}
